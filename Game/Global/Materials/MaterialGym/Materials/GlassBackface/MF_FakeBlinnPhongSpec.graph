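# Material function that adds a fake Blinn-Phong sun specular highlight to input material attributes, apparently for glass.
# A half-vector from sky atmosphere light and camera directions drives a powered normal dot product, masked to one face side via TwoSidedSign and FlipNormals.
# The tinted highlight, optionally colored by sun light, adds to Emissive; its strength adds to Opacity and OpacityMask, zeroed for shadows.
# A blend lerp and texture sample remain unconnected.
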MaterialFunction /Game/Global/Materials/MaterialGym/Materials/GlassBackface/MF_FakeBlinnPhongSpec
graph {nodes [
  n1 = NamedRerouteUsage()
  n2 = Constant()
  n3 = GetMaterialAttributes(None=FunctionInput(Preview=SetMaterialAttributes!(MaterialAttributes, OpacityMask=@n2, Opacity=@n2) {attributeSetTypes [#1, #2]}) {bUsePreviewValueAsDefault=true, inputName=Material, inputType=FunctionInput_MaterialAttributes}) {attributeGetTypes [#3, #1, #2]}
  n4 = Add(A=@n1, B=@n3.EmissiveColor)
  n5 = NamedRerouteUsage()
  n6 = Add(A=Saturate(None=@n3.OpacityMask), B=@n5)
  n7 = Add(A=Saturate(None=@n3.Opacity), B=@n5)
  FunctionOutput(None=SetMaterialAttributes!(MaterialAttributes=@n3.MaterialAttributes, EmissiveColor=@n4, OpacityMask=Saturate(None=@n6), Opacity=Saturate(None=@n7)) {attributeSetTypes [#3, #1, #2]})
  LinearInterpolate(A=@n4, B=LinearInterpolate(A=@n3.EmissiveColor, B=@n1, Alpha=@n5), Alpha=ScalarParameter() {parameterName=BlendSpecColor})
  n8 = Saturate(None=ScalarParameter() {group=FakeSpec, parameterName=FlipNormals})
  n9 = Saturate(None=Multiply(A=LinearInterpolate(Alpha=@n8) {constA=1, constB=-1}, B=TwoSidedSign()))
  n10 = Add(A=SkyAtmosphereLightDirection(), B=NamedRerouteDeclaration(None=CameraVectorWS()) {name=ViewDir, nodeColor=<rgba 0 0.25 1 1>})
  n11 = Saturate(None=DotProduct(A=PixelNormalWS(), B=Normalize(VectorInput=@n10)))
  n12 = Multiply(
   A=Power(Base=Multiply(A=@n9, B=@n11), Exp=ScalarParameter() {defaultValue=64, parameterName=SpecularShininess}) {constExponent=256}
   B=ScalarParameter() {defaultValue=1, parameterName=SpecularOpacityStrength}
  )
  NamedRerouteDeclaration(None=ShadowReplace(Default=Saturate(None=@n12), Shadow=Constant())) {name=BlinnPhongSpecOpacity, nodeColor=<rgba 0.1484 0 1 1>}
  n13 = Multiply(
   A=ScalarParameter() {defaultValue=1, parameterName=SpecularColorStrength}
   B=VectorParameter() {parameterName=SpecularColor, defaultValue=<rgba 1 1 1 1>}.RGB
  )
  n14 = ComponentMask(None=CollectionParameter() {collection=&FortniteMaterialParameters, parameterName=SunLightColor}) {b=true, g=true, r=true}
  n15 = ScalarParameter() {defaultValue=1, parameterName=SunLightColorInfluence}
  NamedRerouteDeclaration(None=Multiply(A=Multiply(A=@n13, B=@n12), B=LinearInterpolate(B=@n14, Alpha=@n15) {constA=1})) {name=BlinnPhongSpecColor, nodeColor=<rgba 0.1484 0 1 1>}
  n16 = TextureCoordinate()
  TextureSample!(UVs, Tex, "DDX(UVs)"=DDX(Value=@n16), "DDY(UVs)"=DDY(Value=@n16), Apply View MipBias) {mipValueMode=TMVM_Derivative}
 ]}
