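# Material function that fakes glass by overriding incoming material attributes with a two-tone fresnel look on one side of the mesh.
# A saturated fresnel term and paired Inner/Outer parameters provide base color, metallic, specular, roughness and emissive.
# The opacity mask uses TwoSidedSign with FlipNormals, keeping backfaces by default, while ShadowReplace keeps shadows solid.
# The Alpha input drives the blend between original and glass attributes.
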MaterialFunction /Game/Global/Materials/MaterialGym/Materials/GlassBackface/MF_FakeGlassBackside_02
graph {nodes [
  n1 = NamedRerouteUsage()
  n2 = SetMaterialAttributes!(
   MaterialAttributes=@n1
   BaseColor=NamedRerouteUsage()
   Metallic=NamedRerouteUsage()
   Specular=NamedRerouteUsage()
   Roughness=NamedRerouteUsage()
   OpacityMask=NamedRerouteUsage()
   EmissiveColor=NamedRerouteUsage()
  ) {attributeSetTypes [
    #1
    #2
    #3
    #4
    #5
    #6
   ]}
  FunctionOutput(None=BlendMaterialAttributes(A=@n1, B=@n2, Alpha=NamedRerouteUsage()))
  NamedRerouteDeclaration(None=FunctionInput(Preview=SetMaterialAttributes!(MaterialAttributes)) {bUsePreviewValueAsDefault=true, inputType=FunctionInput_MaterialAttributes}) {name=IN_Material, nodeColor=<rgba 0 0.6484 1 1>}
  n3 = VectorParameter() {group=FakeGlass, parameterName=EmissiveInner, sortPriority=60, defaultValue=<rgba 0 0 0 1>}
  n4 = VectorParameter() {group=FakeGlass, parameterName=EmissiveOuter, sortPriority=61, defaultValue=<rgba 0.9948 0.9274 0.7927 1>}
  n5 = NamedRerouteUsage()
  NamedRerouteDeclaration(None=Multiply(
    A=LinearInterpolate(A=@n3.RGB, B=@n4.RGB, Alpha=@n5)
    B=ScalarParameter() {defaultValue=1, group=FakeGlass, parameterName=EmissiveStrength, sortPriority=62}
   )) {name=OUT_Emissive, nodeColor=<rgba 0.6406 1 0 1>}
  NamedRerouteDeclaration(None=LinearInterpolate(
    A=ScalarParameter() {defaultValue=0.32, group=FakeGlass, parameterName=RoughnessInner, sortPriority=30}
    B=ScalarParameter() {defaultValue=0.2, group=FakeGlass, parameterName=RoughnessOuter, sortPriority=31}
    Alpha=NamedRerouteUsage()
   )) {name=OUT_Roughness, nodeColor=<rgba 1 0 0.2578 1>}
  n6 = FunctionInput() {bUsePreviewValueAsDefault=true, inputName=Alpha, inputType=FunctionInput_Scalar, sortPriority=1, previewValue=<1 0 0 1>}
  NamedRerouteDeclaration(None=@n6) {name=IN_Alpha, nodeColor=<rgba 1 0 0.1641 1>}
  n7 = PixelNormalWS()
  n8 = ScalarParameter() {defaultValue=2, group=FakeGlass, parameterName=FresnelExponent, sortPriority=10}
  n9 = StaticBool()
  n10 = ScalarParameter() {group=FakeGlass, parameterName=FresnelDarkLevel, sortPriority=11}
  n11 = ScalarParameter() {defaultValue=1, group=FakeGlass, parameterName=FresnelBrightLevel, sortPriority=12}
  NamedRerouteDeclaration(None=Saturate(None=MaterialFunctionCall(Normal Vector=@n7, Power=@n8, Use Cheap contrast=@n9, Cheap contrast dark=@n10, Cheap contrast bright=@n11) {materialFunction=&Fresnel_Function}.Result)) {name=Fresnel_01, nodeColor=<rgba 0 1 0.4844 1>}
  n12 = ScalarParameter() {defaultValue=1, group=FakeGlass, parameterName=FlipNormals, sortPriority=0}
  n13 = Multiply(A=LinearInterpolate(Alpha=Saturate(None=@n12)) {constA=1, constB=-1}, B=TwoSidedSign())
  NamedRerouteDeclaration(None=ShadowReplace(Default=Saturate(None=@n13), Shadow=Constant() {r=1})) {name=OpacityMask, nodeColor=<rgba 1 0 0.375 1>}
  NamedRerouteDeclaration(None=LinearInterpolate(
    A=VectorParameter() {group=FakeGlass, parameterName=ColorInner, sortPriority=20, defaultValue=<rgba 0.8594 0.9303 1 1>}.RGB
    B=VectorParameter() {group=FakeGlass, parameterName=ColorOuter, sortPriority=21, defaultValue=<rgba 1 1 1 0>}.RGB
    Alpha=NamedRerouteUsage()
   )) {name=BaseColor, nodeColor=<rgba 1 0.7031 0 1>}
  NamedRerouteDeclaration(None=LinearInterpolate(
    A=ScalarParameter() {defaultValue=0.7, group=FakeGlass, parameterName=MetallicInner, sortPriority=50}
    B=ScalarParameter() {defaultValue=1, group=FakeGlass, parameterName=MetallicOuter, sortPriority=51}
    Alpha=NamedRerouteUsage()
   )) {name=Metallic, nodeColor=<rgba 1 0 0.4453 1>}
  NamedRerouteDeclaration(None=LinearInterpolate(
    A=ScalarParameter() {defaultValue=0.2, group=FakeGlass, parameterName=SpecularInner, sortPriority=40}
    B=ScalarParameter() {defaultValue=1, group=FakeGlass, parameterName=SpecularOuter, sortPriority=41}
    Alpha=NamedRerouteUsage()
   )) {name=OUT_Specular, nodeColor=<rgba 0.7109 1 0 1>}
 ]}
