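# Material function that adds a fake sun specular highlight to incoming material attributes, apparently for glass.
# It reflects the sky atmosphere light direction about the pixel normal, compares it to negated camera direction, sharpens with a power, and limits it to one face side, flippable by parameter.
# The highlight, tinted by a parameter color or collection sunlight color, adds to Emissive.
# It also raises Opacity and OpacityMask, except in shadow passes via ShadowReplace.
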MaterialFunction /Game/Global/Materials/MaterialGym/Materials/GlassBackface/MF_FakeSpec
graph {nodes [
  n1 = NamedRerouteUsage()
  n2 = Constant()
  n3 = GetMaterialAttributes(None=FunctionInput(Preview=SetMaterialAttributes!(MaterialAttributes, OpacityMask=@n2, Opacity=@n2) {attributeSetTypes [#1, #2]}) {bUsePreviewValueAsDefault=true, inputName=Material, inputType=FunctionInput_MaterialAttributes}) {attributeGetTypes [#3, #1, #2]}
  n4 = NamedRerouteUsage()
  n5 = Add(A=Saturate(None=@n3.OpacityMask), B=@n4)
  n6 = Add(A=Saturate(None=@n3.Opacity), B=@n4)
  FunctionOutput(None=SetMaterialAttributes!(
    MaterialAttributes=@n3.MaterialAttributes
    EmissiveColor=Add(A=@n1, B=@n3.EmissiveColor)
    OpacityMask=Saturate(None=@n5)
    Opacity=Saturate(None=@n6)
   ) {attributeSetTypes [#3, #1, #2]})
  n7 = Multiply(A=ScalarParameter() {defaultValue=1, group=FakeSpec, parameterName=FakeSpecOpacity}, B=NamedRerouteUsage()) {constB=0.4}
  NamedRerouteDeclaration(None=ShadowReplace(Default=Saturate(None=@n7), Shadow=Constant())) {name=OUT_OpacityMask, nodeColor=<rgba 1 0 0.0703 1>}
  n8 = LinearInterpolate(Alpha=Saturate(None=ScalarParameter() {group=FakeSpec, parameterName=FlipNormals})) {constA=1, constB=-1}
  n9 = TwoSidedSign()
  n10 = ComponentMask(None=SkyAtmosphereLightDirection()) {b=true, g=true, r=true}
  n11 = PixelNormalWS()
  n12 = Multiply(A=Multiply(A=DotProduct(A=@n10, B=@n11)) {constB=2}, B=@n11)
  n13 = DotProduct(
   A=Normalize(VectorInput=Subtract(A=@n12, B=@n10))
   B=Multiply(A=MaterialFunctionCall() {materialFunction=&CameraDirectionVector}.Result) {constB=-1}
  )
  n14 = Multiply(
   A=Saturate(None=Multiply(A=@n8, B=@n9))
   B=Power(Base=Max(A=@n13) {constB=0}, Exp=ScalarParameter() {defaultValue=10, group=FakeSpec, parameterName=FakeSpecPow})
  )
  n15 = LinearInterpolate(
   A=VectorParameter() {group=FakeSpec, parameterName=FakeSpecColor, defaultValue=<rgba 1 1 1 1>}.RGB
   B=ComponentMask(None=CollectionParameter() {collection=&FortniteMaterialParameters, parameterName=SunLightColor}) {b=true, g=true, r=true}
   Alpha=ScalarParameter() {group=FakeSpec, parameterName=UseSunlightColor}
  )
  n16 = ScalarParameter() {defaultValue=0.4, group=FakeSpec, parameterName=FakeSpecIntensity}
  NamedRerouteDeclaration(None=Multiply(A=NamedRerouteDeclaration(None=@n14) {name=FakeSpecOpacity, nodeColor=<rgba 0 0.1328 1 1>}, B=Multiply(A=@n15, B=@n16))) {name=OUT_Emissive, nodeColor=<rgba 0 0.9531 1 1>}
 ]}
