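# M_BackfaceGlass is a Material Gym master material that, judging by its names, fakes the back side of glass.
# It is masked, two-sided and dithered rather than blended, using material attributes, with usage enabled for skeletal, morph-target and deformer meshes.
# MF_FakeGlassBackside_02 builds the attributes, and static switch UseFakeSpec optionally routes them through MF_FakeBlinnPhongSpec for fake specular.
# A separate MF_FakeSpec node is left unconnected.
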Material /Game/Global/Materials/MaterialGym/Materials/GlassBackface/M_BackfaceGlass
graph {
 properties {
  bAutomaticallySetUsageInEditor=false
  bDisableDepthTest=true
  bEnableResponsiveAA=true
  bUseMaterialAttributes=true
  bUsedWithMeshDeformer=true
  bUsedWithMorphTargets=true
  bUsedWithSkeletalMesh=true
  blendMode=BLEND_Masked
  ditherOpacityMask=true
  pixelDepthOffsetMode=PDOM_Legacy
  shadingModel=MSM_FromMaterialExpression
  subsurfaceProfile=&SS_HeroSkin_02
  translucencyDirectionalLightingIntensity=5
  translucencyLightingMode=TLM_Surface
  twoSided=true
  usedShadingModels=Default Lit
 }
 nodes [
  MaterialFunctionCall!(Material) {materialFunction=&/MF_FakeSpec}
  n1 = MaterialFunctionCall!(In, Alpha) {materialFunction=&/MF_FakeGlassBackside_02}
  StaticSwitchParameter(
   True=MaterialFunctionCall!(Material=@n1.Result) {materialFunction=&/MF_FakeBlinnPhongSpec}.Result
   False=@n1.Result
  ) {group=FakeSpec, parameterName=UseFakeSpec, sortPriority=-100}
 ]
}
