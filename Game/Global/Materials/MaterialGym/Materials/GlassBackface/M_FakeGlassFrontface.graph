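# Translucent unlit material in the MaterialGym that fakes the front face of glass, judging by its name.
# A Fresnel band, shaped by two smoothsteps, masks sky-light environment reflections tinted by BaseColor, which add to a DepthColor tint in emissive.
# Opacity adds that band to a thickness term from scene depth minus surface depth.
# The attributes pass through MF_FakeBlinnPhongSpec; several alternative Fresnel and blend nodes are left unused.
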Material /Game/Global/Materials/MaterialGym/Materials/GlassBackface/M_FakeGlassFrontface
graph {
 properties {bUseMaterialAttributes=true, blendMode=BLEND_Translucent, pixelDepthOffsetMode=PDOM_Legacy, shadingModel=MSM_Unlit, usedShadingModels=Unlit}
 nodes [
  n1 = VectorParameter() {parameterName=BaseColor, defaultValue=<rgba 1 1 1 1>}
  n2 = PixelNormalWS()
  n3 = Multiply(
   A=SkyLightEnvMapSample(Direction=ReflectionVectorWS(CustomWorldNormal=@n2), Roughness=ScalarParameter() {parameterName=Roughness})
   B=ScalarParameter() {defaultValue=1, parameterName=ReflectionStrength}
  )
  Add(A=@n1.RGB, B=@n3)
  n4 = ScalarParameter() {defaultValue=3, parameterName=FresnelExponent}
  Fresnel(ExponentIn=@n4, BaseReflectFractionIn=ScalarParameter() {parameterName=FresnelBase}, Normal=PixelNormalWS())
  GetMaterialAttributes() {attributeGetTypes [#1, #2]}
  n5 = VectorParameter() {parameterName=DepthColor, defaultValue=<rgba 1 0.0094 0 1>}
  n6 = ScalarParameter() {defaultValue=0.7, parameterName=OuterEdgeThickness}
  n7 = Subtract(A=@n6, B=ScalarParameter() {defaultValue=0.05, parameterName=OuterEdgeFade})
  n8 = Fresnel() {baseReflectFraction=0, exponent=1}
  n9 = OneMinus(None=SmoothStep(Min=@n7, Max=@n6, Value=@n8))
  n10 = ScalarParameter() {parameterName=InnerEdgeFadeBias}
  n11 = SmoothStep(Min=Constant(), Max=Saturate(None=Subtract(A=@n7, B=@n10)), Value=@n8)
  n12 = Saturate(None=Power(Base=Multiply(A=@n9, B=@n11), Exp=ScalarParameter() {defaultValue=2, parameterName=Exp}))
  n13 = Multiply(A=Multiply(A=@n1.RGB, B=@n3), B=@n12)
  n14 = Add(A=Multiply(A=@n5.RGB, B=NamedRerouteUsage()), B=@n13)
  LinearInterpolate(A=Add(A=@n5.RGB, B=@n13), B=@n14, Alpha=NamedRerouteUsage())
  n15 = Divide(
   A=Subtract(
    A=SceneDepth().R
    B=MaterialFunctionCall!(World Position) {materialFunction=/Engine/Functions/Engine_MaterialFunctions02/Utility/DepthFromWorldPosition}.Result
   )
   B=ScalarParameter() {defaultValue=100, parameterName=Thickness}
  )
  n16 = Multiply(A=@n15, B=ScalarParameter() {defaultValue=1, parameterName=DepthAmount})
  MaterialFunctionCall!(Material=SetMaterialAttributes!(MaterialAttributes, Emissive Color=@n14, Opacity=Add(A=@n12, B=@n16)) {attributeSetTypes [#1, #2]}) {materialFunction=&/MF_FakeBlinnPhongSpec}
  MaterialFunctionCall(Power=@n4) {materialFunction=&Fresnel_Function}
  n17 = Add(A=Multiply(A=Fresnel() {baseReflectFraction=0, exponent=1}) {constB=2}, B=ScalarParameter() {group=FakeGlass, parameterName=FresnelOffset})
  Multiply(A=SmoothStep(Value=Sine(None=@n17)), B=Fresnel())
  Multiply(A=@n2) {constB=-1}
  NamedRerouteDeclaration(None=@n12) {name=FresnelOpacity, nodeColor=<rgba 0.8516 0 1 1>}
  NamedRerouteDeclaration(None=Saturate(None=@n16)) {name=DepthOpacity, nodeColor=<rgba 1 0.2578 0 1>}
  OneMinus(None=@n12)
  SetMaterialAttributes(Emissive Color=@n15, Opacity=Constant() {r=1}) {attributeSetTypes [#1, #2]}
  Subtract(A=@n7, B=ScalarParameter() {defaultValue=0.65, parameterName=InnerEdgeThickness})
 ]
}
