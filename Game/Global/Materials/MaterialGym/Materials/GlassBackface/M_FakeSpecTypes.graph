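# Unlit MaterialGym test material that collects alternative fake specular techniques for glass backface work, with none wired to the material outputs.
# Variants include SchlickPhase on light-versus-normal, GGX specular from the sky atmosphere light, and a standalone Schlick approximation node.
# The most developed branch builds Blinn-Phong from a normalized light-plus-view half vector, tinted by a color parameter and adjustable Fortnite sun color, into the BlinnPhongSpec reroute.
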Material /Game/Global/Materials/MaterialGym/Materials/GlassBackface/M_FakeSpecTypes
graph {
 properties {pixelDepthOffsetMode=PDOM_Legacy, shadingModel=MSM_Unlit, usedShadingModels=Unlit}
 nodes [
  MaterialFunctionCall!(Cosine, Anisotropy) {materialFunction=/Engine/Functions/Engine_MaterialFunctions01/Volumetrics/SchlickPhase}
  MaterialFunctionCall!(IOR=Constant() {r=1}) {materialFunction=/InterchangeAssets/Functions/MF_SchlickApprox}
  MaterialFunctionCall(
   Cosine=DotProduct(A=SkyAtmosphereLightDirection(), B=PixelNormalWS())
   Anisotropy=ScalarParameter() {defaultValue=-0.999, group=FakeTranslucency, parameterName=FT_Phase}
  ) {materialFunction=&SchlickPhase}
  n1 = MaterialFunctionCall!(
   Normal=PixelNormalWS()
   Light Vector=SkyAtmosphereLightDirection()
   Roughness=ScalarParameter() {defaultValue=0.1, parameterName="Roughness (S)"}
  ) {materialFunction=/Engine/Functions/Engine_MaterialFunctions01/Lighting/GGXSpecular}
  Multiply(A=@n1.Result, B=ScalarParameter() {defaultValue=0.02, parameterName=Specular}) {constB=0.1}
  n2 = SkyAtmosphereLightDirection()
  n3 = NamedRerouteDeclaration(None=CameraVectorWS()) {name=ViewDir, nodeColor=<rgba 0 0.25 1 1>}
  n4 = DotProduct(A=PixelNormalWS(), B=Normalize(VectorInput=Add(A=@n2, B=@n3)))
  n5 = Multiply(
   A=Power(Base=Saturate(None=@n4), Exp=ScalarParameter() {defaultValue=64, parameterName=SpecularShininess}) {constExponent=256}
   B=ScalarParameter() {defaultValue=1, parameterName=SpecularStrength}
  )
  n6 = VectorParameter() {parameterName=SpecularColor, defaultValue=<rgba 1 1 1 1>}
  n7 = ComponentMask(None=CollectionParameter() {collection=&FortniteMaterialParameters, parameterName=SunLightColor}) {b=true, g=true, r=true}
  n8 = ScalarParameter() {defaultValue=1, parameterName=SunLightColorInfluence}
  NamedRerouteDeclaration(None=Multiply(A=Multiply(A=@n5, B=@n6.RGB), B=LinearInterpolate(B=@n7, Alpha=@n8) {constA=1})) {name=BlinnPhongSpec, nodeColor=<rgba 0.1484 0 1 1>}
  NamedRerouteUsage()
  Saturate(None=@n1.Result)
 ]
}
