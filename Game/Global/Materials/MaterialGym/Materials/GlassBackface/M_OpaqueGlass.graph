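# M_OpaqueGlass is a two-sided Default Lit master material in the MaterialGym GlassBackface folder, which renders glass with Masked rather than translucent blending.
# It uses material attributes, passing the result of MF_OpaqueGlass through MF_FakeSpec to the output.
# Dithered opacity masking is enabled, and the surface logic lives in those two material functions.
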Material /Game/Global/Materials/MaterialGym/Materials/GlassBackface/M_OpaqueGlass
graph {
 properties {
  bUseMaterialAttributes=true
  blendMode=BLEND_Masked
  ditherOpacityMask=true
  pixelDepthOffsetMode=PDOM_Legacy
  twoSided=true
  usedShadingModels=Default Lit
 }
 nodes [MaterialFunctionCall!(Material=MaterialFunctionCall!(In) {materialFunction=&/MF_OpaqueGlass}.Result) {materialFunction=&/MF_FakeSpec}]
}
